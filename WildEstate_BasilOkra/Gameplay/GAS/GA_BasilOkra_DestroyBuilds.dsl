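; Gameplay ability that destroys low-health building actors found by the BasilOkra vent targeting preset.
; On activation it runs a targeting request at once and starts a looping timer, at a scalable-float rate, to repeat it.
; Each result that casts to a building actor is destroyed if its health is at or below a threshold; others only print debug strings.
; The graph stores the timer handle but never clears it or ends the ability.
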(event Ability|EventActivateAbility
  (bind _returnvalue (Utilities|Time|SetTimerbyEvent (delegate Custom|DoTargetingTrace) (ScalableFloat|GetValueatLevel (Variables|Default|GetDestroyTickRate)) true))
  (Variables|Default|SetTargetingTimer _returnvalue)
  (Variables|Default|SetPlayerPawn (CallFunction|GetAvatarasFortPlayerPawn))
  (CallFunction|DoTargetingTrace))

(event Custom|DoTargetingTrace
  (bind (_asynctaskproxy _targetingrequesthandle) (Ability|Tasks|PerformTargetingRequest "/WildEstate_BasilOkra/Gameplay/TP_BasilOkra_Vent_DestroyBuilds.TP_BasilOkra_Vent_DestroyBuilds")
    (:OnTargetReady
      (for _array_element (TargetingSystem|TargetingResults|GetTargetingResultsActors (GameInstanceSubsystems|GetTargetingSubsystem) _targetingrequesthandle)
        (bind _asbuilding_actor (Utilities|Casting|CastToBuildingActor _array_element)
          (:then
            (if (<= (Health|GetHealth _asbuilding_actor :!class /Script/FortniteGame.BuildingActor) (Variables|Default|GetMaxHealthToAutoDestroy))
              (Actor|DestroyActor _asbuilding_actor)
              (else
                (Development|PrintString "too much health"))))
          (:CastFailed
            (Development|PrintString "not a buildingactor")))))))
